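; State Tree task for the Kor boss that shows a looping pre-fire focus Gameplay Cue on its target while the state is active.
; On enter, it casts the AI controller's pawn to the Kor boss and reads the target actor from the Blackboard, caching both.
; It then adds the tagged cue to the target with the boss as SourceObject.
; On exit, it removes the cue from the cached target if still valid.
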(event EventEnterState
  (params (Transition :type State%20Tree%20Transition%20Result%20Structure%20%28by%20ref%29))
  (bind _aicontroller (Variables|Default|GetAIController))
  (Utilities|IsValid _aicontroller
    (:Is%20Valid
      (bind _asbp_player_pawn_cherry_smoke_kor_boss (Utilities|Casting|CastToBP_PlayerPawn_CherrySmoke_KorBoss (Pawn|GetControlledPawn _aicontroller)))
      (bind _output_get (Variables|Default|SetKorPawn _asbp_player_pawn_cherry_smoke_kor_boss))
      (bind _targetactor (Class|BPFLAICommonFunctions|GetTargetActorfromBlackboard _output_get))
      (Utilities|IsValid _targetactor
        (:Is%20Valid
          (bind _output_get_2 (Variables|Default|SetKorTarget _targetactor))
          (GameplayCue|AddGameplayCueOnActor(Looping) _output_get_2 (Variables|Default|GetPreFireFocusGCNL_Tag) (Ability|GameplayCue|MakeGameplayCueParameters :NormalizedMagnitude 0.0 :RawMagnitude 0.0 :Location "0,0,0" :Normal "0,0,0" :SourceObject (Variables|Default|GetKorPawn) :GameplayEffectLevel 1 :AbilityLevel 1 :bReplicateLocationWhenUsingMinimalRepProxy false)))))))

(event EventExitState
  (params (Transition :type State%20Tree%20Transition%20Result%20Structure%20%28by%20ref%29))
  (bind _kortarget (Variables|Default|GetKorTarget))
  (Utilities|IsValid _kortarget
    (:Is%20Valid
      (GameplayCue|RemoveGameplayCueOnActor(Looping) _kortarget (Variables|Default|GetPreFireFocusGCNL_Tag) (Ability|GameplayCue|MakeGameplayCueParameters 0.0)))))
