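# StateTree for the Kor boss bot that equips a weapon, focuses its target and fires until the magazine is empty.
# After equipping, Focus gets and faces the target, crouches and tracks magazine ammo for its child states.
# PreFireDelay waits a parameterized first-shot delay while aiming down sights and playing a focus-beam cue, then a linked-asset FireWeapon state runs.
# An empty magazine moves to Reload, whose completion ends the tree successfully.
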StateTree /KorBoss/StateTree/FocusAndFireUntilWeaponEmpty_Kor
states [
 Root {
  id=#1
  type=STATE
  children [#2, #3]
  tasks []
  transitions []
 }
 EquipWeapon {
  id=#2
  type=STATE
  tasks [{
    id=#4
    description=Equip Item
    properties {
     iD=#5
     instance {_structType=/Script/FortniteAIServer.FortStateTreeBotEquipItemTaskInstanceData, bValidateAbility=false, controller=null, itemToEquip=null}
     node {
      name=None
      _structType=/Script/FortniteAIServer.FortStateTreeBotEquipItemTask
      bCanEditConsideredForCompletion=true
      bConsideredForCompletion=true
      bHasTaskCompletionDelegateDispatcher=false
      bTaskEnabled=true
      transitionHandlingPriority=Normal
      bindingsBatch {value=65535}
      executionRuntimeTemplateIndex {value=65535}
      instanceDataHandle {index=65535, source=None, stateHandle {index=65535}}
      instanceTemplateIndex {value=65535}
      outputBindingsBatch {value=65535}
     }
    }
   }]
  transitions [
   {id=#6, trigger=ON_STATE_SUCCEEDED, state {id=#3, name=Focus}}
   {id=#7, trigger=ON_STATE_FAILED, state {id=#8, name=Root, linkType=FAILED}}
  ]
 }
 Focus {
  id=#3
  type=STATE
  children [#9, #10, #11]
  tasks [
   {
    id=#12
    description=Get Target Actor
    properties {
     iD=#13
     instance {_structType=&FortStateTreeGetTargetActorTaskInstanceData}
     node {name=None, _structType=&FortStateTreeGetTargetActorTask}
    }
   }
   {
    id=#14
    description="(Fortnite) Focus a specific Actor and/or point/location"
    properties {
     iD=#15
     instance {
      _structType=/Script/FortniteAIServer.FortAthenaFocusAtStateTreeTaskInstanceData
      actor=null
      bSetBackOnExit=true
      bShouldSetFocus=true
      focusActor=null
      focusActorOffset=<0 0 0>
      focusWorldPoint=<0 0 0>
     }
     node {
      name=None
      _structType=/Script/FortniteAIServer.FortAthenaFocusAtStateTreeTask
      bCanEditConsideredForCompletion=true
      bConsideredForCompletion=true
      bHasTaskCompletionDelegateDispatcher=false
      bTaskEnabled=true
      transitionHandlingPriority=Normal
      bindingsBatch {value=65535}
      executionRuntimeTemplateIndex {value=65535}
      instanceDataHandle {index=65535, source=None, stateHandle {index=65535}}
      instanceTemplateIndex {value=65535}
      outputBindingsBatch {value=65535}
     }
    }
   }
   {
    id=#16
    description=Bot Crouch Task
    properties {
     iD=#17
     instance {_structType=/Script/FortniteAIServer.FortStateTreeBotCrouchTaskInstanceData, bShouldCrouch=true, botController=null}
     node {
      name=None
      _structType=/Script/FortniteAIServer.FortStateTreeBotCrouchTask
      bCanEditConsideredForCompletion=true
      bConsideredForCompletion=true
      bHasTaskCompletionDelegateDispatcher=false
      bTaskEnabled=true
      transitionHandlingPriority=Normal
      bindingsBatch {value=65535}
      executionRuntimeTemplateIndex {value=65535}
      instanceDataHandle {index=65535, source=None, stateHandle {index=65535}}
      instanceTemplateIndex {value=65535}
      outputBindingsBatch {value=65535}
     }
    }
   }
   {
    id=#18
    description=Bot Get Current Weapon
    properties {
     iD=#19
     instance {_structType=/Script/FortniteAIServer.FortStateTreeBotGetCurrentWeaponTaskInstanceData, botController=null, currentWeapon=null}
     node {
      name=None
      _structType=/Script/FortniteAIServer.FortStateTreeBotGetCurrentWeaponTask
      bCanEditConsideredForCompletion=true
      bConsideredForCompletion=true
      bHasTaskCompletionDelegateDispatcher=false
      bTaskEnabled=true
      transitionHandlingPriority=Normal
      bindingsBatch {value=65535}
      executionRuntimeTemplateIndex {value=65535}
      instanceDataHandle {index=65535, source=None, stateHandle {index=65535}}
      instanceTemplateIndex {value=65535}
      outputBindingsBatch {value=65535}
     }
    }
   }
   {
    id=#20
    description=Get Weapon Magazine Ammo
    properties {
     iD=#21
     instance {_structType=/Script/FortniteAIServer.FortStateTreeGetWeaponMagazineAmmoTaskInstanceData, ammoInMagazine=0, weapon=null}
     node {
      name=None
      _structType=/Script/FortniteAIServer.FortStateTreeGetWeaponMagazineAmmoTask
      bCanEditConsideredForCompletion=true
      bConsideredForCompletion=true
      bHasTaskCompletionDelegateDispatcher=false
      bTaskEnabled=true
      transitionHandlingPriority=Normal
      bindingsBatch {value=65535}
      executionRuntimeTemplateIndex {value=65535}
      instanceDataHandle {index=65535, source=None, stateHandle {index=65535}}
      instanceTemplateIndex {value=65535}
      outputBindingsBatch {value=65535}
     }
    }
   }
  ]
  transitions [{
    id=#22
    trigger=ON_TICK
    conditions [{
      id=#23
      description="Get Weapon Magazine Ammo.Ammo in Magazine == 0"
      properties {
       iD=#24
       instance {_structType=/Script/StateTreeModule.StateTreeCompareIntConditionInstanceData, left=0, right=0}
       node {
        name=None
        _structType=/Script/StateTreeModule.StateTreeCompareIntCondition
        bInvert=false
        deltaIndent=0
        evaluationMode=Evaluated
        operand=And
        operator=Equal
        bindingsBatch {value=65535}
        executionRuntimeTemplateIndex {value=65535}
        instanceDataHandle {index=65535, source=None, stateHandle {index=65535}}
        instanceTemplateIndex {value=65535}
        outputBindingsBatch {value=65535}
       }
      }
     }]
    state {id=#11, name=Reload}
   }]
 }
 PreFireDelay {
  id=#9
  type=STATE
  tasks [
   {
    id=#25
    description=Delay Parameters.First Shot Delay
    properties {
     iD=#26
     instance {_structType=&StateTreeDelayTaskInstanceData, duration=1}
     node {name=None, _structType=&StateTreeDelayTask}
    }
   }
   {
    id=#27
    description="Set Weapon Targeting (ADS)"
    properties {
     iD=#28
     instance {_structType=/Script/FortniteAIServer.FortStateTreeSetWeaponTargetingTaskInstanceData, bShouldActivateTargeting=true, weapon=null}
     node {
      name=None
      _structType=/Script/FortniteAIServer.FortStateTreeSetWeaponTargetingTask
      bCanEditConsideredForCompletion=true
      bConsideredForCompletion=true
      bHasTaskCompletionDelegateDispatcher=false
      bTaskEnabled=true
      transitionHandlingPriority=Normal
      bindingsBatch {value=65535}
      executionRuntimeTemplateIndex {value=65535}
      instanceDataHandle {index=65535, source=None, stateHandle {index=65535}}
      instanceTemplateIndex {value=65535}
      outputBindingsBatch {value=65535}
     }
    }
   }
   {
    id=#29
    description=Calls a GCNL for Kor's prefire focus beam.
    properties {
     iD=#30
     instanceObject=&self:StateTreeEditorData_0.StateTreeState_3.BP_KorBoss_PreFireFocus_C_0
     node {
      name=None
      _structType=&StateTreeBlueprintTaskWrapper
      taskClass=&/BP_KorBoss_PreFireFocus.BP_KorBoss_PreFireFocus_C
     }
    }
   }
  ]
  transitions [{id=#31, trigger=ON_STATE_COMPLETED, state {id=#10, name=FireWeapon}}]
 }
 FireWeapon {
  id=#10
  type=LINKED_ASSET
  transitions [{
    id=#32
    priority=null
    trigger=ON_TICK
    conditions [{
      id=#33
      description="Get Weapon Magazine Ammo.Ammo in Magazine == 0"
      properties {
       executionRuntimeDataObject=null
       expressionIndent=0
       expressionOperand=And
       iD=#34
       instanceObject=null
       executionRuntimeData {}
       instance {_structType=&StateTreeCompareIntConditionInstanceData, left=0, right=0}
       node {
        name=None
        _structType=&StateTreeCompareIntCondition
        bInvert=false
        deltaIndent=0
        evaluationMode=Evaluated
        operand=And
        operator=Equal
        bindingsBatch {value=65535}
        executionRuntimeTemplateIndex {value=65535}
        instanceDataHandle {index=65535, source=None, stateHandle {index=65535}}
        instanceTemplateIndex {value=65535}
        outputBindingsBatch {value=65535}
       }
      }
     }]
    state {id=#11, name=Reload, linkType=GOTO_STATE}
   }]
 }
 Reload {
  id=#11
  type=STATE
  tasks [{
    id=#35
    description=Reload Weapon
    properties {
     iD=#36
     instance {_structType=/Script/FortniteAIServer.FortStateTreeReloadWeaponTaskInstanceData, weapon=null}
     node {
      name=None
      _structType=/Script/FortniteAIServer.FortStateTreeReloadWeaponTask
      bCanEditConsideredForCompletion=true
      bConsideredForCompletion=true
      bHasTaskCompletionDelegateDispatcher=false
      bTaskEnabled=true
      transitionHandlingPriority=Normal
      bindingsBatch {value=65535}
      executionRuntimeTemplateIndex {value=65535}
      instanceDataHandle {index=65535, source=None, stateHandle {index=65535}}
      instanceTemplateIndex {value=65535}
      outputBindingsBatch {value=65535}
     }
    }
   }]
  transitions [{id=#37, trigger=ON_STATE_COMPLETED, state {id=#8, name=Root, linkType=SUCCEEDED}}]
 }
]
